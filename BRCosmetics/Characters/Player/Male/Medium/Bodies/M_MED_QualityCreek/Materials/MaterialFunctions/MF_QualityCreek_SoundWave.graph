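# Material function for the QualityCreek character body that generates a fake sound-wave signal from panning Perlin noise, remapped between low and high parameters.
# Audio reactivity amplitude switches tiling and speed between active and idle values, with optional stepped time and vertex-color green offsetting the UVs.
# Default/high quality height-blends two noise scales and modulates them with a finer third.
# Mobile and low/medium quality use a cheaper single-sample remap.
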MaterialFunction /BRCosmetics/Characters/Player/Male/Medium/Bodies/M_MED_QualityCreek/Materials/MaterialFunctions/MF_QualityCreek_SoundWave
graph {nodes [
  n1 = MaterialFunctionCall() {materialFunction=/Game/Packages/Fortress_SharedMaterials/Base_Material_Functions/BaseCharacterShaderFunctions/Audio/MF_AudioReactivity}
  n2 = ScalarParameter() {defaultValue=1, group=QualityCreek_SoundWave, parameterName=Fake Signal Tiling Idle}
  n3 = AppendVector(
   A=If(
    A=@n1.Amplitude
    "A > B"=ScalarParameter() {defaultValue=1, group=QualityCreek_SoundWave, parameterName=Fake Signal Tiling}
    "A == B"=@n2
    "A < B"=@n2
   )
   B=Constant()
  )
  n4 = Multiply(A=@n3) {constB=3}
  n5 = NamedRerouteUsage()
  n6 = Multiply(A=@n5, B=@n4)
  n7 = Time()
  n8 = ScalarParameter() {defaultValue=20, group=QualityCreek_SoundWave, parameterName=SteppedTimeFrameRate, sortPriority=41}
  n9 = Round(None=Multiply(A=@n7, B=@n8) {constB=6})
  n10 = ScalarParameter() {defaultValue=2.5, group=QualityCreek_SoundWave, parameterName=Fake Signal Speed Idle}
  n11 = Multiply(
   A=StaticSwitchParameter(True=Divide(A=@n9, B=@n8), False=@n7) {defaultValue=true, group=QualityCreek_SoundWave, parameterName=UseSteppedTime, sortPriority=40}
   B=If(
    A=@n1.Amplitude
    "A > B"=ScalarParameter() {defaultValue=2.5, group=QualityCreek_SoundWave, parameterName=Fake Signal Speed}
    "A == B"=@n10
    "A < B"=@n10
   )
  )
  n12 = VertexColor()
  n13 = TextureSample(UVs=Add(A=Panner(Coordinate=@n6, Time=@n11) {speedX=0.05, speedY=0.12}, B=@n12.G)) {texture=/BRCosmetics/Characters/Player/Male/Medium/Bodies/M_MED_QualityCreek/Textures/T_PerlinRGB_01_M}
  n14 = Multiply(A=@n13.G) {constB=0.6}
  n15 = Multiply(A=@n5, B=@n3)
  n16 = TextureSample(UVs=Add(A=Panner(Coordinate=@n15, Time=@n11) {speedX=0.05, speedY=0.12}, B=@n12.G)) {texture=&T_PerlinRGB_01_M}
  n17 = Multiply(A=Subtract(A=@n14, B=@n16.G)) {constB=0.5}
  n18 = Constant() {r=0.5}
  n19 = Saturate(None=Add(A=Divide(A=@n17, B=@n18)) {constB=0.5})
  n20 = LinearInterpolate(A=@n14, B=@n16.G, Alpha=@n19)
  n21 = Multiply(A=OneMinus(None=@n19), B=Multiply(A=@n19, B=@n18))
  n22 = Multiply(A=@n4) {constB=2}
  n23 = Add(A=Panner(Coordinate=Multiply(A=@n5, B=@n22), Time=@n11) {speedX=0.1, speedY=0.2}, B=@n12.G)
  n24 = Multiply(
   A=Add(A=Subtract(A=@n20, B=@n21)) {constB=0.5}
   B=LinearInterpolate(B=TextureSample(UVs=@n23) {texture=&T_PerlinRGB_01_M}.B) {constA=1, constAlpha=0.3}
  )
  n25 = LinearInterpolate(Alpha=@n13.R) {constA=0.3}
  n26 = ShadingPathSwitch(Default=@n24, Deferred=@n24, Mobile=@n25) {desc=Migrated from FeatureLevelSwitch}
  FunctionOutput(None=LinearInterpolate(
    A=ScalarParameter() {defaultValue=-0.5, group=QualityCreek_SoundWave, parameterName=Fake Signal Low}
    B=ScalarParameter() {defaultValue=1, group=QualityCreek_SoundWave, parameterName=Fake Signal High}
    Alpha=QualitySwitch(Default=@n26, Low=@n25, High=@n26, Medium=@n25, Epic=@n26)
   )) {outputName=Fake Signal, sortPriority=13}
  NamedRerouteDeclaration(None=FunctionInput(Preview=TextureCoordinate()) {bUsePreviewValueAsDefault=true, inputName=UVs, inputType=FunctionInput_Vector2}) {name=UV INPUT, nodeColor=<rgba 1 0.1464 0 1>}
 ]}
