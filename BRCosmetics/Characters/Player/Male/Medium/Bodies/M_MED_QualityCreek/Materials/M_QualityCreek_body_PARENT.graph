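# Parent body material for the QualityCreek character skin, adding an audio-reactive glowing soundwave to the shared 2023 character PreFX/PostFX shader stack.
# MF_QualityCreek_SoundWave supplies a fake signal, which audio amplitude scales, falling back to a clamped, subtle idle wave without audio.
# The signal and its negation vertically offset two panning line-texture samples, forming a mirrored waveform.
# The cyan-tinted lines, masked by base emissive, add to it, and time of day dims the total at night.
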Material /BRCosmetics/Characters/Player/Male/Medium/Bodies/M_MED_QualityCreek/Materials/M_QualityCreek_body_PARENT
graph {
 properties {
  bUseMaterialAttributes=true
  bUsedWithMeshDeformer=true
  bUsedWithMorphTargets=true
  bUsedWithNanite=true
  bUsedWithSkeletalMesh=true
  pixelDepthOffsetMode=PDOM_Legacy
  shadingModel=MSM_FromMaterialExpression
  usedShadingModels=Default Lit | Subsurface | Clear Coat | Subsurface Profile | Cloth
 }
 nodes [
  n1 = MaterialFunctionCall() {materialFunction=&MF_CharacterParent_2023_PreFX}
  n2 = GetMaterialAttributes(None=MaterialFunctionCall(Material=@n1.Result, M=@n1.Result, Pre Skinned Local Position=@n1.Result, Distance Blend=@n1.Result, Fresnel=@n1.Result) {materialFunction=&MF_CharacterParent_2023_PostFX}.Result) {attributeGetTypes [#1]}
  n3 = MaterialFunctionCall() {materialFunction=/Game/Packages/Fortress_SharedMaterials/Base_Material_Functions/BaseCharacterShaderFunctions/Audio/MF_AudioReactivity}
  n4 = MaterialFunctionCall!(UVs=MaterialFunctionCall(UVs=TextureCoordinate(), Texture Scale=Constant() {r=1}) {materialFunction=&ScaleUVsByCenter}.UVs) {materialFunction=/BRCosmetics/Characters/Player/Male/Medium/Bodies/M_MED_QualityCreek/Materials/MaterialFunctions/MF_QualityCreek_SoundWave}
  n5 = Sine(None=Clamp(None=@n4.Fake Signal) {maxDefault=0.03, minDefault=-0.03})
  n6 = If(A=@n3.Amplitude, "A > B"=@n4.Fake Signal, "A == B"=@n5, "A < B"=@n5)
  n7 = ComponentMask(None=@n6) {r=true}
  n8 = ScalarParameter() {defaultValue=0.085, group=QualityCreek_SoundWave, parameterName=ReactivityIntensityIdle}
  n9 = If(
   A=@n3.Amplitude
   "A > B"=Multiply(A=@n3.Amplitude, B=ScalarParameter() {defaultValue=0.085, group=QualityCreek_SoundWave, parameterName=ReactivityIntensity})
   "A == B"=@n8
   "A < B"=@n8
  )
  n10 = Add(A=Panner() {speedX=1}, B=AppendVector(A=Constant(), B=Multiply(A=@n7, B=@n9) {constB=0.1}))
  n11 = Constant2Vector() {g=-0.365}
  n12 = Constant() {r=1}
  n13 = ScalarParameter() {defaultValue=0.1, group=QualityCreek_SoundWave, parameterName=AudioLineThickness}
  n14 = TextureSample(UVs=MaterialFunctionCall(UVs=Add(A=@n10, B=@n11), Texture Scale=AppendVector(A=@n12, B=@n13)) {materialFunction=&ScaleUVsByCenter}.UVs) {
   samplerSource=SSM_Clamp_WorldGroupSettings
   samplerType=SAMPLERTYPE_Masks
   texture=/BRCosmetics/Characters/Player/Male/Medium/Bodies/M_MED_QualityCreek/Textures/T_QualityCreek_AudioLine
  }
  n15 = ComponentMask(None=Multiply(A=@n6) {constB=-1}) {r=true}
  n16 = Add(A=Panner() {speedX=1}, B=AppendVector(A=Constant(), B=Multiply(A=@n15, B=@n9) {constB=0.1}))
  n17 = Constant2Vector() {g=-0.365}
  n18 = Constant() {r=1}
  n19 = ScalarParameter() {defaultValue=0.1, group=QualityCreek_SoundWave, parameterName=AudioLineThickness}
  n20 = TextureSample(UVs=MaterialFunctionCall(UVs=Add(A=@n16, B=@n17), Texture Scale=AppendVector(A=@n18, B=@n19)) {materialFunction=&ScaleUVsByCenter}.UVs) {
   samplerSource=SSM_Clamp_WorldGroupSettings
   samplerType=SAMPLERTYPE_Masks
   texture=&T_QualityCreek_AudioLine
  }
  n21 = Multiply(A=Multiply(A=@n2.Emissive Color, B=Add(A=@n14.R, B=@n20.R)), B=Constant3Vector() {constant=<rgba 0 4.2656 5 1>}) {constB=5}
  n22 = Constant() {r=1}
  n23 = Constant() {r=0.5}
  SetMaterialAttributes(
   MaterialAttributes=@n2.MaterialAttributes
   Emissive Color=Multiply(
    A=Add(A=@n2.Emissive Color, B=@n21)
    B=MaterialFunctionCall(Value_Morning=@n22, Value_Day=@n22, Value_Evening=@n22, Value_Night=@n23) {materialFunction=&Scalar_Time_of_Day}.Result
   )
  ) {attributeSetTypes [#1]}
 ]
}
