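# Secondary-motion AnimBlueprint for the QualityCreek male body cosmetic, layering gear animation over the pose copied from the parent mesh.
# Constraints and additive bone offsets adjust shell, straps, pads and pouches, then spring-driven AnimDynamics swings hand ribbons.
# The QuietPeanuts Control Rig follows, then backpack and cape attach bones are constrained to the shell.
# EventGraph converts skydiving, parachute, melee and weapon-targeting states into alphas; pistol targeting goes unused.
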AnimBlueprint /BRCosmetics/Characters/Player/Male/Medium/Bodies/M_MED_QualityCreek/Meshes/M_MED_QualityCreek_AnimBP
target_skeleton=&/M_MED_QualityCreek_Skeleton
graphs [
 AnimGraph {nodes [
   Output Pose(Result=Use cached pose 'backPackCache'()) {layerGroup=None}
   n1 = Control Rig!(
    Source=Use cached pose 'preCRCache'()
    Alpha
    Is Playing Melee Alpha:double=Get Is Playing Melee Alpha() -> (Is Playing Melee Alpha:double)
    Is Skydiving Alpha:double=Get Is Skydiving Alpha() -> (Is Skydiving Alpha:double)
   ) -> (Pose) {
    alphaBoolBlend {blendOption=Linear}
    controlRigAssetReference {blueprintRigClass=/BRCosmetics/Characters/Player/Male/Medium/Bodies/M_MED_QuietPeanuts/Meshes/M_MED_QuietPeanuts_CtrlRig.M_MED_QuietPeanuts_CtrlRig_C}
   }
   Save cached pose 'CRCache'(Pose=@n1) {cacheName=CRCache}
   n2 = "Constraint - dyn_strap_bk_l_MSTR (PP)"!(
    ComponentPose=Local To Component(LocalPose=Copy Pose From Mesh() {bUseAttachedParent=true})
    Alpha=1
    ConstraintWeights_0=0.75
    ConstraintWeights_1:float=0.25
   ) -> (Pose) {
    alphaBoolBlend {blendOption=Linear}
    boneToModify {boneName=dyn_strap_bk_l_MSTR}
    constraintSetup [{targetBone {boneName=dyn_turtleShellTp_l_01}}, {targetBone {boneName=clavicle_l}}]
    constraintWeights [1,1]
   }
   n3 = "Constraint - dyn_strap_bk_r_MSTR (PP)"!(ComponentPose=@n2, Alpha=1, ConstraintWeights_0=0.75, ConstraintWeights_1:float=0.25) -> (Pose) {
    alphaBoolBlend {blendOption=Linear}
    boneToModify {boneName=dyn_strap_bk_r_MSTR}
    constraintSetup [{targetBone {boneName=dyn_turtleShellTp_r_01}}, {targetBone {boneName=clavicle_r}}]
    constraintWeights [1,1]
   }
   n4 = "Constraint - dyn_thighPouch_l_01 (PPPPPPPP)"!(
    ComponentPose=@n3
    Alpha=1
    ConstraintWeights_0=0.4152
    ConstraintWeights_1:float=0.4014
    ConstraintWeights_2:float=0.1023
    ConstraintWeights_3:float=0.0449
    ConstraintWeights_4:float=0.0148
    ConstraintWeights_5:float=0.0096
    ConstraintWeights_6:float=0.0061
    ConstraintWeights_7:float=0.0058
   ) -> (Pose) {
    alphaBoolBlend {blendOption=Linear}
    boneToModify {boneName=dyn_thighPouch_l_01}
    constraintSetup [
     {targetBone {boneName=thigh_twist_01_l}}
     {targetBone {boneName=pelvis}}
     {targetBone {boneName=deform_glute_l}}
     {targetBone {boneName=deform_groin_l}}
     {targetBone {boneName=dyn_turtleShell_01}}
     {targetBone {boneName=thigh_l}}
     {targetBone {boneName=spine_01}}
     {targetBone {boneName=spine_02}}
    ]
    constraintWeights [1,1,1,1,1,1,1,1]
   }
   n5 = "Constraint - dyn_legBuckle_r_01 (PPPPPPPP)"!(
    ComponentPose=@n4
    Alpha=1
    ConstraintWeights_0=0.4154
    ConstraintWeights_1:float=0.4045
    ConstraintWeights_2:float=0.101
    ConstraintWeights_3:float=0.0434
    ConstraintWeights_4:float=0.0143
    ConstraintWeights_5:float=0.0094
    ConstraintWeights_6:float=0.0061
    ConstraintWeights_7:float=0.0059
   ) -> (Pose) {
    alphaBoolBlend {blendOption=Linear}
    boneToModify {boneName=dyn_legBuckle_r_01}
    constraintSetup [
     {targetBone {boneName=thigh_twist_01_r}}
     {targetBone {boneName=pelvis}}
     {targetBone {boneName=deform_glute_r}}
     {targetBone {boneName=deform_groin_r}}
     {targetBone {boneName=dyn_turtleShell_01}}
     {targetBone {boneName=thigh_r}}
     {targetBone {boneName=spine_01}}
     {targetBone {boneName=spine_02}}
    ]
    constraintWeights [1,1,1,1,1,1,1,1]
   }
   n6 = "Constraint - deform_knee_pad_l (PP)"!(ComponentPose=@n5, Alpha=1, ConstraintWeights_0=0.5, ConstraintWeights_1:float=0.5) -> (Pose) {
    alphaBoolBlend {blendOption=Linear}
    boneToModify {boneName=deform_knee_pad_l}
    constraintSetup [{targetBone {boneName=deform_knee_in_base_l}}, {targetBone {boneName=calf_l}}]
    constraintWeights [1,1]
   }
   n7 = "Constraint - deform_knee_pad_r (PP)"!(ComponentPose=@n6, Alpha=1, ConstraintWeights_0=0.5, ConstraintWeights_1:float=0.5) -> (Pose) {
    alphaBoolBlend {blendOption=Linear}
    boneToModify {boneName=deform_knee_pad_r}
    constraintSetup [{targetBone {boneName=deform_knee_in_base_r}}, {targetBone {boneName=calf_r}}]
    constraintWeights [1,1]
   }
   n8 = "Constraint - deform_elbow_pad_l (PP)"!(ComponentPose=@n7, Alpha=1, ConstraintWeights_0=0.5, ConstraintWeights_1:float=0.5) -> (Pose) {
    alphaBoolBlend {blendOption=Linear}
    boneToModify {boneName=deform_elbow_pad_l}
    constraintSetup [{targetBone {boneName=deform_elbow_in_base_l}}, {targetBone {boneName=lowerarm_l}}]
    constraintWeights [1,1]
   }
   n9 = "Constraint - deform_elbow_pad_r (PP)"!(ComponentPose=@n8, Alpha=1, ConstraintWeights_0=0.5, ConstraintWeights_1:float=0.5) -> (Pose) {
    alphaBoolBlend {blendOption=Linear}
    boneToModify {boneName=deform_elbow_pad_r}
    constraintSetup [{targetBone {boneName=deform_elbow_in_base_r}}, {targetBone {boneName=lowerarm_r}}]
    constraintWeights [1,1]
   }
   n10 = "Constraint - dyn_turtleShell_MSTR (PPPPP)"!(
    ComponentPose=@n9
    Alpha=Get Is not Parachute Opened() -> (Is not Parachute Opened:double)
    ConstraintWeights_0=0.2
    ConstraintWeights_1:float=0.2
    ConstraintWeights_2:float=0.2
    ConstraintWeights_3:float=0.2
    ConstraintWeights_4:float=0.2
   ) -> (Pose) {
    alphaBoolBlend {blendOption=Linear}
    boneToModify {boneName=dyn_turtleShell_MSTR}
    constraintSetup [
     {targetBone {boneName=spine_01}}
     {targetBone {boneName=spine_02}}
     {targetBone {boneName=spine_03}}
     {targetBone {boneName=spine_04}}
     {targetBone {boneName=spine_05}}
    ]
    constraintWeights [1,1,1,1,1]
   }
   n11 = "Constraint - dyn_turtleShell_MSTR (PPPPP)"(
    ComponentPose=@n10
    Alpha=Get Is Parachute Opened() -> (Is Parachute Opened:double)
    ConstraintWeights_0=0
    ConstraintWeights_1=0
    ConstraintWeights_2=0
    ConstraintWeights_3=0
    ConstraintWeights_4=1
   ) {
    alphaBoolBlend {blendOption=Linear}
    boneToModify {boneName=dyn_turtleShell_MSTR}
    constraintSetup [
     {targetBone {boneName=spine_01}}
     {targetBone {boneName=spine_02}}
     {targetBone {boneName=spine_03}}
     {targetBone {boneName=spine_04}}
     {targetBone {boneName=spine_05}}
    ]
    constraintWeights [1,1,1,1,1]
   }
   n12 = Get Is Rifle Targeting Alpha() -> (Is Rifle Targeting Alpha:double)
   n13 = "Transform (Modify) Bone - Bone: ChestPouch_01_jnt"(ComponentPose=@n11, Alpha=@n12, Translation="-0.5,-1,2", Rotation="20,0,0", Scale="0.75,0.75,0.75") {
    rotationMode=BMM_Additive
    rotationSpace=BCS_BoneSpace
    scaleMode=BMM_Additive
    scaleSpace=BCS_BoneSpace
    translationMode=BMM_Additive
    translationSpace=BCS_BoneSpace
    alphaBoolBlend {blendOption=Linear}
    alphaScaleBiasClamp {bInterpResult=true, interpSpeedDecreasing=6}
    boneToModify {boneName=ChestPouch_01_jnt}
    rotation {pitch=20}
    scale=<0.75 0.75 0.75>
    translation=<-0.5 -1 2>
   }
   n14 = "Transform (Modify) Bone - Bone: ChestPouch_02_jnt"(ComponentPose=@n13, Alpha=@n12, Translation="-0.5,1,2", Rotation="0,0,0", Scale="0.75,0.75,0.75") {
    rotationMode=BMM_Additive
    rotationSpace=BCS_BoneSpace
    scaleMode=BMM_Additive
    scaleSpace=BCS_BoneSpace
    translationMode=BMM_Additive
    translationSpace=BCS_BoneSpace
    alphaBoolBlend {blendOption=Linear}
    alphaScaleBiasClamp {bInterpResult=true, interpSpeedDecreasing=6}
    boneToModify {boneName=ChestPouch_02_jnt}
    scale=<0.75 0.75 0.75>
    translation=<-0.5 1 2>
   }
   n15 = Get Is Skydiving Alpha()
   n16 = "Transform (Modify) Bone - Bone: dyn_turtleShellTp_01"(ComponentPose=@n14, Alpha=@n15, Translation="0,0,-3", Rotation="0,0,0", Scale="1,1,0.75") {
    rotationMode=BMM_Additive
    rotationSpace=BCS_BoneSpace
    scaleMode=BMM_Additive
    scaleSpace=BCS_BoneSpace
    translationMode=BMM_Additive
    translationSpace=BCS_BoneSpace
    alphaBoolBlend {blendOption=Linear}
    alphaScaleBiasClamp {bInterpResult=true, interpSpeedIncreasing=4}
    boneToModify {boneName=dyn_turtleShellTp_01}
    scale {z=0.75}
    translation {z=-3}
   }
   n17 = "Transform (Modify) Bone - Bone: dyn_strap_l_MSTR"(ComponentPose=@n16, Alpha=@n15, Translation="0,0,1", Rotation="15,-15,-15", Scale="1,1,1") {
    rotationMode=BMM_Additive
    rotationSpace=BCS_BoneSpace
    scaleMode=BMM_Additive
    scaleSpace=BCS_BoneSpace
    translationMode=BMM_Additive
    translationSpace=BCS_BoneSpace
    alphaBoolBlend {blendOption=Linear}
    alphaScaleBiasClamp {bInterpResult=true, interpSpeedIncreasing=4}
    boneToModify {boneName=dyn_strap_l_MSTR}
    rotation=<rot 15 -15 -15>
    translation {z=1}
   }
   n18 = "Transform (Modify) Bone - Bone: dyn_strap_r_MSTR"(ComponentPose=@n17, Alpha=@n15, Translation="0,0,1", Rotation="15,15,15", Scale="1,1,1") {
    rotationMode=BMM_Additive
    rotationSpace=BCS_BoneSpace
    scaleMode=BMM_Additive
    scaleSpace=BCS_BoneSpace
    translationMode=BMM_Additive
    translationSpace=BCS_BoneSpace
    alphaBoolBlend {blendOption=Linear}
    alphaScaleBiasClamp {bInterpResult=true, interpSpeedIncreasing=4}
    boneToModify {boneName=dyn_strap_r_MSTR}
    rotation=<rot 15 15 15>
    translation {z=1}
   }
   n19 = "Constraint - dyn_turtleShellTp_01 (P)"(ComponentPose=@n18, Alpha=Get Is not Parachute Opened(), ConstraintWeights_0=1) {
    alphaBoolBlend {blendOption=Linear}
    boneToModify {boneName=dyn_turtleShellTp_01}
    constraintSetup [{targetBone {boneName=spine_04}}]
    constraintWeights [1]
   }
   n20 = "Constraint - dyn_turtleShellTp_01 (P)"(ComponentPose=@n19, Alpha=Get Is Parachute Opened(), ConstraintWeights_0=1) {
    alphaBoolBlend {blendOption=Linear}
    boneToModify {boneName=dyn_turtleShellTp_01}
    constraintSetup [{targetBone {boneName=spine_05}}]
    constraintWeights [1]
   }
   n21 = "Transform (Modify) Bone - Bone: dyn_strap_bk_l_MSTR"(ComponentPose=@n20, Alpha=@n15, Translation="-3,0,1", Rotation="0,15,0", Scale="1,1,1") {
    rotationMode=BMM_Additive
    rotationSpace=BCS_BoneSpace
    scaleMode=BMM_Additive
    scaleSpace=BCS_BoneSpace
    translationMode=BMM_Additive
    translationSpace=BCS_BoneSpace
    alphaBoolBlend {blendOption=Linear}
    alphaScaleBiasClamp {bInterpResult=true, interpSpeedIncreasing=4}
    boneToModify {boneName=dyn_strap_bk_l_MSTR}
    rotation {yaw=15}
    translation {x=-3, z=1}
   }
   n22 = "Transform (Modify) Bone - Bone: dyn_strap_bk_r_MSTR"(ComponentPose=@n21, Alpha=@n15, Translation="-3,0,1", Rotation="0,-15,0", Scale="1,1,1") {
    rotationMode=BMM_Additive
    rotationSpace=BCS_BoneSpace
    scaleMode=BMM_Additive
    scaleSpace=BCS_BoneSpace
    translationMode=BMM_Additive
    translationSpace=BCS_BoneSpace
    alphaBoolBlend {blendOption=Linear}
    alphaScaleBiasClamp {bInterpResult=true, interpSpeedIncreasing=4}
    boneToModify {boneName=dyn_strap_bk_r_MSTR}
    rotation {yaw=-15}
    translation {x=-3, z=1}
   }
   n23 = Anim Dynamics - Bone: dyn_01_handribbon_l_01_jnt(ComponentPose=@n22, Alpha=0.75, ExternalForce="0,0,0") {
    angularDampingOverride=1
    angularSpringConstant=200
    bAngularSpring=true
    bOverrideAngularDamping=true
    bUseGravityOverride=true
    simulationSpace=RootRelative
    alphaBoolBlend {blendOption=Linear}
    boundBone {boneName=dyn_01_handribbon_l_01_jnt}
    componentAppliedLinearAccClamp=<6000 6000 6000>
    componentLinearAccScale=<0.5 0.5 0.5>
    componentLinearVelScale=<0.5 0.5 0.5>
    gravityOverride {z=-980}
    physicsBodyDefinitions [{
      boundBone {boneName=dyn_01_handribbon_l_01_jnt}
      boxExtents=<10 10 10>
      constraintSetup {angularLimitsMax=<45 45 45>, angularLimitsMin=<-45 -45 -45>, angularTarget=<1 0 0>}
      localJointOffset=<4 0 0>
     }]
   }
   n24 = Anim Dynamics - Bone: dyn_01_handribbon_l_02_jnt(ComponentPose=@n23, Alpha=1, ExternalForce="0,0,0") {
    angularDampingOverride=1
    angularSpringConstant=200
    bAngularSpring=true
    bOverrideAngularDamping=true
    bUseGravityOverride=true
    simulationSpace=RootRelative
    alphaBoolBlend {blendOption=Linear}
    boundBone {boneName=dyn_01_handribbon_l_02_jnt}
    componentAppliedLinearAccClamp=<6000 6000 6000>
    componentLinearAccScale=<0.5 0.5 0.5>
    componentLinearVelScale=<0.5 0.5 0.5>
    gravityOverride {z=-980}
    physicsBodyDefinitions [{
      boundBone {boneName=dyn_01_handribbon_l_02_jnt}
      boxExtents=<10 10 10>
      constraintSetup {angularLimitsMax=<45 45 45>, angularLimitsMin=<-45 -45 -45>, angularTarget=<1 0 0>}
      localJointOffset=<4 0 0>
     }]
   }
   n25 = Anim Dynamics - Bone: dyn_01_handribbon_l_03_jnt(ComponentPose=@n24, Alpha=1, ExternalForce="0,0,0") {
    angularDampingOverride=1
    angularSpringConstant=200
    bAngularSpring=true
    bOverrideAngularDamping=true
    bUseGravityOverride=true
    simulationSpace=RootRelative
    alphaBoolBlend {blendOption=Linear}
    boundBone {boneName=dyn_01_handribbon_l_03_jnt}
    componentAppliedLinearAccClamp=<6000 6000 6000>
    componentLinearAccScale=<0.5 0.5 0.5>
    componentLinearVelScale=<0.5 0.5 0.5>
    gravityOverride {z=-980}
    physicsBodyDefinitions [{
      boundBone {boneName=dyn_01_handribbon_l_03_jnt}
      boxExtents=<10 10 10>
      constraintSetup {angularLimitsMax=<45 45 45>, angularLimitsMin=<-45 -45 -45>, angularTarget=<1 0 0>}
      localJointOffset=<4 0 0>
     }]
   }
   n26 = Anim Dynamics - Bone: dyn_01_handribbon_l_04_jnt(ComponentPose=@n25, Alpha=1, ExternalForce="0,0,0") {
    angularDampingOverride=1
    angularSpringConstant=200
    bAngularSpring=true
    bOverrideAngularDamping=true
    bUseGravityOverride=true
    simulationSpace=RootRelative
    alphaBoolBlend {blendOption=Linear}
    boundBone {boneName=dyn_01_handribbon_l_04_jnt}
    componentAppliedLinearAccClamp=<6000 6000 6000>
    componentLinearAccScale=<0.5 0.5 0.5>
    componentLinearVelScale=<0.5 0.5 0.5>
    gravityOverride {z=-980}
    physicsBodyDefinitions [{
      boundBone {boneName=dyn_01_handribbon_l_04_jnt}
      boxExtents=<10 10 10>
      constraintSetup {angularLimitsMax=<45 45 45>, angularLimitsMin=<-45 -45 -45>, angularTarget=<1 0 0>}
      localJointOffset=<4 0 0>
     }]
   }
   n27 = Anim Dynamics - Bone: dyn_01_handribbon_l_05_jnt(ComponentPose=@n26, Alpha=1, ExternalForce="0,0,0") {
    angularDampingOverride=1
    angularSpringConstant=200
    bAngularSpring=true
    bOverrideAngularDamping=true
    bUseGravityOverride=true
    simulationSpace=RootRelative
    alphaBoolBlend {blendOption=Linear}
    boundBone {boneName=dyn_01_handribbon_l_05_jnt}
    componentAppliedLinearAccClamp=<6000 6000 6000>
    componentLinearAccScale=<0.5 0.5 0.5>
    componentLinearVelScale=<0.5 0.5 0.5>
    gravityOverride {z=-980}
    physicsBodyDefinitions [{
      boundBone {boneName=dyn_01_handribbon_l_05_jnt}
      boxExtents=<10 10 10>
      constraintSetup {angularLimitsMax=<45 45 45>, angularLimitsMin=<-45 -45 -45>, angularTarget=<1 0 0>}
      localJointOffset=<4 0 0>
     }]
   }
   n28 = Anim Dynamics - Bone: dyn_02_handribbon_l_01_jnt(ComponentPose=@n27, Alpha=0.75, ExternalForce="0,0,0") {
    angularDampingOverride=1
    angularSpringConstant=200
    bAngularSpring=true
    bOverrideAngularDamping=true
    bUseGravityOverride=true
    simulationSpace=RootRelative
    alphaBoolBlend {blendOption=Linear}
    boundBone {boneName=dyn_02_handribbon_l_01_jnt}
    componentAppliedLinearAccClamp=<6000 6000 6000>
    componentLinearAccScale=<0.5 0.5 0.5>
    componentLinearVelScale=<0.5 0.5 0.5>
    gravityOverride {z=-980}
    physicsBodyDefinitions [{
      boundBone {boneName=dyn_02_handribbon_l_01_jnt}
      boxExtents=<10 10 10>
      constraintSetup {angularLimitsMax=<45 45 45>, angularLimitsMin=<-45 -45 -45>, angularTarget=<1 0 0>}
      localJointOffset=<4 0 0>
     }]
   }
   n29 = Anim Dynamics - Bone: dyn_02_handribbon_l_02_jnt(ComponentPose=@n28, Alpha=1, ExternalForce="0,0,0") {
    angularDampingOverride=1
    angularSpringConstant=200
    bAngularSpring=true
    bOverrideAngularDamping=true
    bUseGravityOverride=true
    simulationSpace=RootRelative
    alphaBoolBlend {blendOption=Linear}
    boundBone {boneName=dyn_02_handribbon_l_02_jnt}
    componentAppliedLinearAccClamp=<6000 6000 6000>
    componentLinearAccScale=<0.5 0.5 0.5>
    componentLinearVelScale=<0.5 0.5 0.5>
    gravityOverride {z=-980}
    physicsBodyDefinitions [{
      boundBone {boneName=dyn_02_handribbon_l_02_jnt}
      boxExtents=<10 10 10>
      constraintSetup {angularLimitsMax=<45 45 45>, angularLimitsMin=<-45 -45 -45>, angularTarget=<1 0 0>}
      localJointOffset=<4 0 0>
     }]
   }
   n30 = Anim Dynamics - Bone: dyn_02_handribbon_l_03_jnt(ComponentPose=@n29, Alpha=1, ExternalForce="0,0,0") {
    angularDampingOverride=1
    angularSpringConstant=200
    bAngularSpring=true
    bOverrideAngularDamping=true
    bUseGravityOverride=true
    simulationSpace=RootRelative
    alphaBoolBlend {blendOption=Linear}
    boundBone {boneName=dyn_02_handribbon_l_03_jnt}
    componentAppliedLinearAccClamp=<6000 6000 6000>
    componentLinearAccScale=<0.5 0.5 0.5>
    componentLinearVelScale=<0.5 0.5 0.5>
    gravityOverride {z=-980}
    physicsBodyDefinitions [{
      boundBone {boneName=dyn_02_handribbon_l_03_jnt}
      boxExtents=<10 10 10>
      constraintSetup {angularLimitsMax=<45 45 45>, angularLimitsMin=<-45 -45 -45>, angularTarget=<1 0 0>}
      localJointOffset=<4 0 0>
     }]
   }
   n31 = Anim Dynamics - Bone: dyn_02_handribbon_l_04_jnt(ComponentPose=@n30, Alpha=1, ExternalForce="0,0,0") {
    angularDampingOverride=1
    angularSpringConstant=200
    bAngularSpring=true
    bOverrideAngularDamping=true
    bUseGravityOverride=true
    simulationSpace=RootRelative
    alphaBoolBlend {blendOption=Linear}
    boundBone {boneName=dyn_02_handribbon_l_04_jnt}
    componentAppliedLinearAccClamp=<6000 6000 6000>
    componentLinearAccScale=<0.5 0.5 0.5>
    componentLinearVelScale=<0.5 0.5 0.5>
    gravityOverride {z=-980}
    physicsBodyDefinitions [{
      boundBone {boneName=dyn_02_handribbon_l_04_jnt}
      boxExtents=<10 10 10>
      constraintSetup {angularLimitsMax=<45 45 45>, angularLimitsMin=<-45 -45 -45>, angularTarget=<1 0 0>}
      localJointOffset=<4 0 0>
     }]
   }
   n32 = Anim Dynamics - Bone: dyn_02_handribbon_l_05_jnt(ComponentPose=@n31, Alpha=1, ExternalForce="0,0,0") {
    angularDampingOverride=1
    angularSpringConstant=200
    bAngularSpring=true
    bOverrideAngularDamping=true
    bUseGravityOverride=true
    simulationSpace=RootRelative
    alphaBoolBlend {blendOption=Linear}
    boundBone {boneName=dyn_02_handribbon_l_05_jnt}
    componentAppliedLinearAccClamp=<6000 6000 6000>
    componentLinearAccScale=<0.5 0.5 0.5>
    componentLinearVelScale=<0.5 0.5 0.5>
    gravityOverride {z=-980}
    physicsBodyDefinitions [{
      boundBone {boneName=dyn_02_handribbon_l_05_jnt}
      boxExtents=<10 10 10>
      constraintSetup {angularLimitsMax=<45 45 45>, angularLimitsMin=<-45 -45 -45>, angularTarget=<1 0 0>}
      localJointOffset=<4 0 0>
     }]
   }
   n33 = Anim Dynamics - Bone: dyn_01_handribbon_r_01_jnt(ComponentPose=@n32, Alpha=0.75, ExternalForce="0,0,0") {
    angularDampingOverride=1
    angularSpringConstant=200
    bAngularSpring=true
    bOverrideAngularDamping=true
    bUseGravityOverride=true
    simulationSpace=RootRelative
    alphaBoolBlend {blendOption=Linear}
    boundBone {boneName=dyn_01_handribbon_r_01_jnt}
    componentAppliedLinearAccClamp=<6000 6000 6000>
    componentLinearAccScale=<0.5 0.5 0.5>
    componentLinearVelScale=<0.5 0.5 0.5>
    gravityOverride {z=-980}
    physicsBodyDefinitions [{
      boundBone {boneName=dyn_01_handribbon_r_01_jnt}
      boxExtents=<10 10 10>
      constraintSetup {angularLimitsMax=<45 45 45>, angularLimitsMin=<-45 -45 -45>, angularTarget=<1 0 0>}
      localJointOffset=<4 0 0>
     }]
   }
   n34 = Anim Dynamics - Bone: dyn_01_handribbon_r_02_jnt(ComponentPose=@n33, Alpha=1, ExternalForce="0,0,0") {
    angularDampingOverride=1
    angularSpringConstant=200
    bAngularSpring=true
    bOverrideAngularDamping=true
    bUseGravityOverride=true
    simulationSpace=RootRelative
    alphaBoolBlend {blendOption=Linear}
    boundBone {boneName=dyn_01_handribbon_r_02_jnt}
    componentAppliedLinearAccClamp=<6000 6000 6000>
    componentLinearAccScale=<0.5 0.5 0.5>
    componentLinearVelScale=<0.5 0.5 0.5>
    gravityOverride {z=-980}
    physicsBodyDefinitions [{
      boundBone {boneName=dyn_01_handribbon_r_02_jnt}
      boxExtents=<10 10 10>
      constraintSetup {angularLimitsMax=<45 45 45>, angularLimitsMin=<-45 -45 -45>, angularTarget=<1 0 0>}
      localJointOffset=<4 0 0>
     }]
   }
   n35 = Anim Dynamics - Bone: dyn_01_handribbon_r_03_jnt(ComponentPose=@n34, Alpha=1, ExternalForce="0,0,0") {
    angularDampingOverride=1
    angularSpringConstant=200
    bAngularSpring=true
    bOverrideAngularDamping=true
    bUseGravityOverride=true
    simulationSpace=RootRelative
    alphaBoolBlend {blendOption=Linear}
    boundBone {boneName=dyn_01_handribbon_r_03_jnt}
    componentAppliedLinearAccClamp=<6000 6000 6000>
    componentLinearAccScale=<0.5 0.5 0.5>
    componentLinearVelScale=<0.5 0.5 0.5>
    gravityOverride {z=-980}
    physicsBodyDefinitions [{
      boundBone {boneName=dyn_01_handribbon_r_03_jnt}
      boxExtents=<10 10 10>
      constraintSetup {angularLimitsMax=<45 45 45>, angularLimitsMin=<-45 -45 -45>, angularTarget=<1 0 0>}
      localJointOffset=<4 0 0>
     }]
   }
   n36 = Anim Dynamics - Bone: dyn_01_handribbon_r_04_jnt(ComponentPose=@n35, Alpha=1, ExternalForce="0,0,0") {
    angularDampingOverride=1
    angularSpringConstant=200
    bAngularSpring=true
    bOverrideAngularDamping=true
    bUseGravityOverride=true
    simulationSpace=RootRelative
    alphaBoolBlend {blendOption=Linear}
    boundBone {boneName=dyn_01_handribbon_r_04_jnt}
    componentAppliedLinearAccClamp=<6000 6000 6000>
    componentLinearAccScale=<0.5 0.5 0.5>
    componentLinearVelScale=<0.5 0.5 0.5>
    gravityOverride {z=-980}
    physicsBodyDefinitions [{
      boundBone {boneName=dyn_01_handribbon_r_04_jnt}
      boxExtents=<10 10 10>
      constraintSetup {angularLimitsMax=<45 45 45>, angularLimitsMin=<-45 -45 -45>, angularTarget=<1 0 0>}
      localJointOffset=<4 0 0>
     }]
   }
   n37 = Anim Dynamics - Bone: dyn_01_handribbon_r_05_jnt(ComponentPose=@n36, Alpha=1, ExternalForce="0,0,0") {
    angularDampingOverride=1
    angularSpringConstant=200
    bAngularSpring=true
    bOverrideAngularDamping=true
    bUseGravityOverride=true
    simulationSpace=RootRelative
    alphaBoolBlend {blendOption=Linear}
    boundBone {boneName=dyn_01_handribbon_r_05_jnt}
    componentAppliedLinearAccClamp=<6000 6000 6000>
    componentLinearAccScale=<0.5 0.5 0.5>
    componentLinearVelScale=<0.5 0.5 0.5>
    gravityOverride {z=-980}
    physicsBodyDefinitions [{
      boundBone {boneName=dyn_01_handribbon_r_05_jnt}
      boxExtents=<10 10 10>
      constraintSetup {angularLimitsMax=<45 45 45>, angularLimitsMin=<-45 -45 -45>, angularTarget=<1 0 0>}
      localJointOffset=<4 0 0>
     }]
   }
   n38 = Anim Dynamics - Bone: dyn_02_handribbon_r_01_jnt(ComponentPose=@n37, Alpha=0.75, ExternalForce="0,0,0") {
    angularDampingOverride=1
    angularSpringConstant=200
    bAngularSpring=true
    bOverrideAngularDamping=true
    bUseGravityOverride=true
    simulationSpace=RootRelative
    alphaBoolBlend {blendOption=Linear}
    boundBone {boneName=dyn_02_handribbon_r_01_jnt}
    componentAppliedLinearAccClamp=<6000 6000 6000>
    componentLinearAccScale=<0.5 0.5 0.5>
    componentLinearVelScale=<0.5 0.5 0.5>
    gravityOverride {z=-980}
    physicsBodyDefinitions [{
      boundBone {boneName=dyn_02_handribbon_r_01_jnt}
      boxExtents=<10 10 10>
      constraintSetup {angularLimitsMax=<45 45 45>, angularLimitsMin=<-45 -45 -45>, angularTarget=<1 0 0>}
      localJointOffset=<4 0 0>
     }]
   }
   n39 = Anim Dynamics - Bone: dyn_02_handribbon_r_02_jnt(ComponentPose=@n38, Alpha=1, ExternalForce="0,0,0") {
    angularDampingOverride=1
    angularSpringConstant=200
    bAngularSpring=true
    bOverrideAngularDamping=true
    bUseGravityOverride=true
    simulationSpace=RootRelative
    alphaBoolBlend {blendOption=Linear}
    boundBone {boneName=dyn_02_handribbon_r_02_jnt}
    componentAppliedLinearAccClamp=<6000 6000 6000>
    componentLinearAccScale=<0.5 0.5 0.5>
    componentLinearVelScale=<0.5 0.5 0.5>
    gravityOverride {z=-980}
    physicsBodyDefinitions [{
      boundBone {boneName=dyn_02_handribbon_r_02_jnt}
      boxExtents=<10 10 10>
      constraintSetup {angularLimitsMax=<45 45 45>, angularLimitsMin=<-45 -45 -45>, angularTarget=<1 0 0>}
      localJointOffset=<4 0 0>
     }]
   }
   n40 = Anim Dynamics - Bone: dyn_02_handribbon_r_03_jnt(ComponentPose=@n39, Alpha=1, ExternalForce="0,0,0") {
    angularDampingOverride=1
    angularSpringConstant=200
    bAngularSpring=true
    bOverrideAngularDamping=true
    bUseGravityOverride=true
    simulationSpace=RootRelative
    alphaBoolBlend {blendOption=Linear}
    boundBone {boneName=dyn_02_handribbon_r_03_jnt}
    componentAppliedLinearAccClamp=<6000 6000 6000>
    componentLinearAccScale=<0.5 0.5 0.5>
    componentLinearVelScale=<0.5 0.5 0.5>
    gravityOverride {z=-980}
    physicsBodyDefinitions [{
      boundBone {boneName=dyn_02_handribbon_r_03_jnt}
      boxExtents=<10 10 10>
      constraintSetup {angularLimitsMax=<45 45 45>, angularLimitsMin=<-45 -45 -45>, angularTarget=<1 0 0>}
      localJointOffset=<4 0 0>
     }]
   }
   n41 = Anim Dynamics - Bone: dyn_02_handribbon_r_04_jnt(ComponentPose=@n40, Alpha=1, ExternalForce="0,0,0") {
    angularDampingOverride=1
    angularSpringConstant=200
    bAngularSpring=true
    bOverrideAngularDamping=true
    bUseGravityOverride=true
    simulationSpace=RootRelative
    alphaBoolBlend {blendOption=Linear}
    boundBone {boneName=dyn_02_handribbon_r_04_jnt}
    componentAppliedLinearAccClamp=<6000 6000 6000>
    componentLinearAccScale=<0.5 0.5 0.5>
    componentLinearVelScale=<0.5 0.5 0.5>
    gravityOverride {z=-980}
    physicsBodyDefinitions [{
      boundBone {boneName=dyn_02_handribbon_r_04_jnt}
      boxExtents=<10 10 10>
      constraintSetup {angularLimitsMax=<45 45 45>, angularLimitsMin=<-45 -45 -45>, angularTarget=<1 0 0>}
      localJointOffset=<4 0 0>
     }]
   }
   n42 = Anim Dynamics - Bone: dyn_02_handribbon_r_05_jnt(ComponentPose=@n41, Alpha=1, ExternalForce="0,0,0") {
    angularDampingOverride=1
    angularSpringConstant=200
    bAngularSpring=true
    bOverrideAngularDamping=true
    bUseGravityOverride=true
    simulationSpace=RootRelative
    alphaBoolBlend {blendOption=Linear}
    boundBone {boneName=dyn_02_handribbon_r_05_jnt}
    componentAppliedLinearAccClamp=<6000 6000 6000>
    componentLinearAccScale=<0.5 0.5 0.5>
    componentLinearVelScale=<0.5 0.5 0.5>
    gravityOverride {z=-980}
    physicsBodyDefinitions [{
      boundBone {boneName=dyn_02_handribbon_r_05_jnt}
      boxExtents=<10 10 10>
      constraintSetup {angularLimitsMax=<45 45 45>, angularLimitsMin=<-45 -45 -45>, angularTarget=<1 0 0>}
      localJointOffset=<4 0 0>
     }]
   }
   Save cached pose 'preCRCache'(Pose=Component To Local(ComponentPose=@n42)) {cacheName=preCRCache}
   n43 = "Constraint - attach_backpack (PP)"!(ComponentPose=Local To Component(LocalPose=Use cached pose 'CRCache'()), Alpha=1, ConstraintWeights_0=0.5, ConstraintWeights_1:float=0.5) -> (Pose) {
    alphaBoolBlend {blendOption=Linear}
    boneToModify {boneName=attach_backpack}
    constraintSetup [{targetBone {boneName=dyn_turtleShellTp_01}}, {targetBone {boneName=dyn_turtleShell_MSTR}}]
    constraintWeights [1,1]
   }
   n44 = "Constraint - attach_cape (PP)"!(ComponentPose=@n43, Alpha=1, ConstraintWeights_0=0.5, ConstraintWeights_1:float=0.5) -> (Pose) {
    alphaBoolBlend {blendOption=Linear}
    boneToModify {boneName=attach_cape}
    constraintSetup [{targetBone {boneName=dyn_turtleShellTp_01}}, {targetBone {boneName=dyn_turtleShell_MSTR}}]
    constraintWeights [1,1]
   }
   Save cached pose 'backPackCache'(Pose=Component To Local(ComponentPose=@n44)) {cacheName=backPackCache}
  ]}
 EventGraph {nodes [
   Event BlueprintUpdateAnimation()
   > n1 = Sequence!(execute) -> (then_0, then_1, then_2:Exec, then_3:Exec, then_4:Exec)
   > Set Is Skydiving Alpha(Is Skydiving Alpha:double="To Float (Boolean)"(InBool=Get bIsSkyDiving())) -> (Output_Get:double)
   Set Is Playing Melee Alpha(execute=@n1.then_1, Is Playing Melee Alpha:double="To Float (Boolean)"(InBool=Get bIsPlayingMeleeAnim())) -> (Output_Get:double)
   Set Is Parachute Opened(execute=@n1.then_2, Is Parachute Opened:double="To Float (Boolean)"(InBool=Get bIsParachuteOpened())) -> (Output_Get:double)
   n2 = Get bIsParachuteOpened()
   Set Is not Parachute Opened(execute=@n1.then_3, Is not Parachute Opened:double="To Float (Boolean)"(InBool=NOT Boolean(A=@n2))) -> (Output_Get:double)
   n3 = Sequence(execute=@n1.then_4)
   > n4 = Cast To FortPlayerAnimInstance(Object=GetAnimInstance(self=GetOwnerSkeletalMeshComponent()))
   > Set isPistol(isPistol=false)
   > Set isRifle(isRifle:bool=false) -> (Output_Get:bool)
   > n5 = Switch on EFortWeaponCoreAnimation!(execute, Selection:EFortWeaponCoreAnimation Enum=Get WeaponCoreAnim(self=@n4)) -> (
    Melee:Exec
    Pistol:Exec
    Shotgun:Exec
    PaperBlueprint:Exec
    Rifle:Exec
    MeleeOneHand:Exec
    MachinePistol:Exec
    RocketLauncher:Exec
    GrenadeLauncher:Exec
    GoingCommando:Exec
    AssaultRifle:Exec
    TacticalShotgun:Exec
    SniperRifle:Exec
    TrapPlacement:Exec
    ShoulderLauncher:Exec
    AbilityDecoTool:Exec
    Crossbow:Exec
    C4:Exec
    RemoteControl:Exec
    DualWield:Exec
    AR_BullPup:Exec
    AR_ForwardGrip:Exec
    MedPackPaddles:Exec
    SMG_P90:Exec
    AR_DrumGun:Exec
    Consumable_Small:Exec
    Consumable_Large:Exec
    Balloon:Exec
    MountedTurret:Exec
    CreativeTool:Exec
    ExplosiveBow:Exec
    AshtonIndigo:Exec
    AshtonChicago:Exec
    MeleeDualWield:Exec
    Unarmed:Exec
   )
   Set isRifle(execute=[@n5.AssaultRifle, @n5.Rifle, @n5.RocketLauncher, @n5.Shotgun, @n5.SniperRifle, @n5.TacticalShotgun], isRifle=true)
   Set isPistol(execute=@n5.Pistol, isPistol=true)
   n6 = Get bIsTargeting()
   n7 = Get isRifle() -> (isRifle:bool)
   Set Is Rifle Targeting Alpha(execute=@n3.then_1, Is Rifle Targeting Alpha:double="To Float (Boolean)"(InBool=AND Boolean(A=@n7, B=@n6))) -> (Output_Get:double)
   n8 = Get bIsTargeting()
   n9 = Get isPistol() -> (isPistol:bool)
   Set Is Pistol Targeting Alpha(execute=@n3.then_2, Is Pistol Targeting Alpha:float="To Float (Boolean)"(InBool=AND Boolean(A=@n9, B=@n8))) -> (Output_Get:float)
  ]}
]
variables [
 Is not Parachute Opened {type=number}
 Is Parachute Opened {type=number}
 Is Pistol Targeting Alpha {type=number}
 Is Playing Melee Alpha {type=number}
 Is Rifle Targeting Alpha {type=number}
 Is Skydiving Alpha {type=number}
 isPistol {type=boolean}
 isRifle {type=boolean}
]
