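# Post-process style animation blueprint for the QualityCreek male face accessory, adding procedural sway and physics, notably to its headband bones.
# AnimGraph copies the parent mesh pose through a Control Rig, additively rotates left and right headband bones, then simulates RigidBody physics relative to neck_01.
# Each update computes state-based gravity through a deprecated helper and two Perlin-noise rotators for the headband.
# A second RigidBody node, borrowing QuietPeanuts physics, is left unconnected.
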AnimBlueprint /BRCosmetics/Characters/Player/Male/Medium/Bodies/M_MED_QualityCreek/Meshes/Parts/M_MED_QualityCreek_FaceAcc_AnimBP
target_skeleton=&/M_MED_QualityCreek_FaceAcc_Skeleton
graphs [
 AnimGraph {nodes [
   n1 = Get GravityOverride_tessel() -> (GravityOverride_tessel:Vector)
   RigidBody!(
    ComponentPose
    Alpha
    OverrideWorldGravity:Vector=@n1
    ExternalForce
    ComponentLinearAccScale:Vector="0.2,0.2,0.2"
    ComponentLinearVelScale:Vector="0.4,0.3,0.3"
   ) -> (Pose) {
    bClampLinearTranslationLimitToRefPose=true
    bOverrideWorldGravity=true
    overridePhysicsAsset=/BRCosmetics/Characters/Player/Male/Medium/Bodies/M_MED_QuietPeanuts/Meshes/Parts/M_MED_QuietPeanuts_FaceAcc_Ground_Physics
    simulationSpace=BaseBoneSpace
    alphaBoolBlend {blendOption=Linear}
    alphaScaleBiasClamp {bInterpResult=true, interpSpeedIncreasing=2}
    baseBoneRef {boneName=neck_02}
    componentAppliedLinearAccClamp=<3000 3000 3000>
    componentLinearAccScale=<0.4 0.3 0.3>
    componentLinearVelScale=<0.4 0.3 0.3>
    simSpaceSettings {
     maxAngularAcceleration=1000
     maxAngularVelocity=3
     maxLinearAcceleration=700
     maxLinearVelocity=500
     velocityScaleZ=0
     worldAlpha=0.25
     externalLinearDragV=<0.2 0.2>
    }
   }
   n2 = Control Rig(Source=Copy Pose From Mesh() {bUseAttachedParent=true}) {
    alphaBoolBlend {blendOption=Linear}
    controlRigAssetReference {blueprintRigClass=&/M_MED_QualityCreek_FaceAcc_CtrlRig.M_MED_QualityCreek_FaceAcc_CtrlRig_C}
   }
   n3 = "Transform (Modify) Bone - Bone: dyn_HeadBand_l_MSTR"(
    ComponentPose=Local To Component(LocalPose=@n2)
    Alpha=0.5
    Translation="0,0,0"
    Rotation=Get TesselRotator_01() -> (TesselRotator_01:Rotator)
    Scale="1,1,1"
   ) {
    rotationMode=BMM_Additive
    rotationSpace=BCS_BoneSpace
    translationSpace=BCS_BoneSpace
    alphaBoolBlend {blendOption=Linear}
    alphaScaleBiasClamp {interpSpeedDecreasing=2, interpSpeedIncreasing=2}
    boneToModify {boneName=dyn_HeadBand_l_MSTR}
   }
   n4 = "Transform (Modify) Bone - Bone: dyn_HeadBand_r_MSTR"(ComponentPose=@n3, Alpha=0.85, Translation="0,0,0", Rotation=Get TesselRotator_02() -> (TesselRotator_02:Rotator), Scale="1,1,1") {
    rotationMode=BMM_Additive
    rotationSpace=BCS_BoneSpace
    translationSpace=BCS_BoneSpace
    alphaBoolBlend {blendOption=Linear}
    alphaScaleBiasClamp {bInterpResult=true, interpSpeedDecreasing=2, interpSpeedIncreasing=2}
    boneToModify {boneName=dyn_HeadBand_r_MSTR}
   }
   n5 = RigidBody!(ComponentPose=@n4, Alpha, OverrideWorldGravity=@n1, ExternalForce) -> (Pose) {
    bClampLinearTranslationLimitToRefPose=true
    bOverrideWorldGravity=true
    overridePhysicsAsset=&/M_MED_QualityCreek_FaceAcc_Physics
    simulationSpace=BaseBoneSpace
    alphaBoolBlend {blendOption=Linear}
    baseBoneRef {boneName=neck_01}
    componentAppliedLinearAccClamp=<3000 3000 3000>
    componentLinearAccScale {x=0.1}
    componentLinearVelScale {x=0.1}
    overrideWorldGravity {z=-980}
    simSpaceSettings {
     maxAngularAcceleration=1000
     maxAngularVelocity=3
     maxLinearAcceleration=700
     maxLinearVelocity=500
     velocityScaleZ=0
     worldAlpha=1
     externalLinearDragV=<0.5 0.5>
    }
   }
   Output Pose(Result=Component To Local(ComponentPose=@n5)) {layerGroup=None}
  ]}
 EventGraph {nodes [
   Event BlueprintUpdateAnimation()
   > n1 = Sequence!(execute) -> (then_0, then_1, then_2:Exec, then_3:Exec)
   n2 = Gravity Override State Function DEPRECATED(
    execute=@n1.then_1
    CharacterPartAnimInstance=Self-Reference()
    GameTimeInSeconds=GetGameTimeInSeconds()
    GravityOverrideParams=Get Gravity Override Params()
   )
   > Set GravityOverride_tessel(GravityOverride_tessel:Vector=@n2) -> (Output_Get:Vector)
   n3 = GetGameTimeInSeconds()
   n4 = RandomRotatorFunction!(
    execute=@n1.then_2
    GameTimeInSeconds:double=@n3
    PerlinScale:double=1
    Frequency:Rotator="5,2.9,0"
    Amplitude:Rotator="30,15,0"
    Offset:Rotator="30,-1.8,0"
    TimeOffset:double=0
   ) -> (then, RandomRotator:Rotator)
   > Set TesselRotator_01(TesselRotator_01:Rotator=@n4) -> (Output_Get:Rotator)
   n5 = RandomRotatorFunction(execute=@n1.then_3, GameTimeInSeconds=@n3, PerlinScale=1, Frequency="4.5,1.6,0", Amplitude="30,15,0", Offset="40,15,0", TimeOffset=0)
   > Set TesselRotator_02(TesselRotator_02:Rotator=@n5) -> (Output_Get:Rotator)
  ]}
]
variables [
 Gravity Override Params {
  type=GravityOverrideParamsStruct
  default {
   emoteOrMelee {gravityOverride=<0 -50 -15>, windAmplitude=<50 -25 100>, windFrequency=<1.5 10 4>, windOffset=<0 -100 -50>}
   front_End {gravityOverride=<0 -50 -15>, windAmplitude=<50 -25 100>, windFrequency=<1.5 10 4>, windOffset=<0 -100 -50>}
   onGround_Moving {gravityOverride=<0 -50 -15>, windAmplitude=<50 -25 100>, windFrequency=<1.8 13 6>, windOffset=<0 -100 -50>}
   onGround_Standing {gravityOverride=<0 -50 -15>, windAmplitude=<50 -25 100>, windFrequency=<1.8 13 6>, windOffset=<0 -100 -50>}
  }
 }
 GravityOverride_tessel {type=Vector}
 TesselRotator_01 {type=Rotator}
 TesselRotator_02 {type=Rotator}
]
